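# Master subsurface material for the Clash V large male player body, built between the shared character PreFX and PostFX functions.
# Vertigo's throbbing function is driven by the Skin FX mask's green channel and a GoopAmount exponent.
# The inverted red thickness channel, tinted and multiplied by the PreFX Gmap, sets subsurface color.
# An optional glider wobble switch adds time-scrolled curl noise over pre-skinned position to world position offset.
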Material /ClashGameplay/Characters/Player/Male/Large/Bodies/M_LRG_Clash_V/Materials/M_ClashV_Master
graph {
 properties {
  bUseMaterialAttributes=true
  bUsedWithMeshDeformer=true
  bUsedWithSkeletalMesh=true
  pixelDepthOffsetMode=PDOM_Legacy
  shadingModel=MSM_Subsurface
  usedShadingModels=Subsurface
 }
 nodes [
  n1 = TextureSampleParameter2D() {parameterName=Skin FX, samplerType=SAMPLERTYPE_Masks, texture=&T_BlackMask}
  n2 = Power(Base=@n1.G, Exp=ScalarParameter() {defaultValue=0.75, parameterName=GoopAmount}) {constExponent=1}
  n3 = MaterialFunctionCall() {materialFunction=&BaseFN_CharacterFunction_PreFX}
  n4 = GetMaterialAttributes(None=MaterialFunctionCall!(Alpha=Saturate(None=@n2), Pre SKinned Local=@n3.Pre skinned Local Position, ThicknessMap, In=@n3.Pre skinned Local Position) {materialFunction=/Game/Characters/Player/Male/Large/Bodies/M_LRG_Vertigo/Materials/MF_Throbbing}.Result) {attributeGetTypes [#1]}
  n5 = OneMinus(None=@n1.R)
  n6 = Multiply(
   A=Clamp(None=Power(Base=@n5), Min=ScalarParameter() {defaultValue=0.15, parameterName=SSS_Min})
   B=VectorParameter() {parameterName=SS COlor, defaultValue=<rgba 0.05 0.0093 0.0078 1>}.RGB
  )
  n7 = MaterialFunctionCall(Float3=Divide(
    A=@n3.Pre skinned Local Position
    B=VectorParameter() {group=BlackMonday, parameterName=DisplacementNoiseScale, defaultValue=<rgba 512 256 256 1>}.RGB
   )) {materialFunction=&BreakOutFloat3Components}
  n8 = Time()
  n9 = ScalarParameter() {defaultValue=1, group=BlackMonday, parameterName=WindFlapSpeed}
  n10 = MaterialFunctionCall(X=@n7.R, Y=Add(A=@n7.G, B=Multiply(A=@n8, B=@n9)), Z=@n7.R) {materialFunction=&MakeFloat3}
  n11 = VectorParameter() {group=BlackMonday, parameterName=VertexDIsplacement_Amt, defaultValue=<rgba 50 50 1 1>}
  n12 = Add(A=@n4.World Position Offset, B=Multiply(A=VectorNoise(World Position=@n10.Result) {noiseFunction=VNF_CurlALU}, B=@n11.RGB))
  MaterialFunctionCall(Material=SetMaterialAttributes!(
    MaterialAttributes=@n4.MaterialAttributes
    Subsurface Color=Multiply(A=@n3.Gmap, B=@n6)
    World Position Offset=StaticSwitchParameter(True=@n12, False=@n4.World Position Offset) {parameterName=useGliderWobble}
   ) {attributeSetTypes [#2, #1]}) {materialFunction=&BaseFN_CharacterFunction_PostFX}
  MaterialFunctionCall()
  Multiply(B=@n11.RGB)
  NamedRerouteDeclaration(None=@n1.R) {name=Thickness Map, nodeColor=<rgba 1 0 0.0937 1>}
  OneMinus(None=@n3.AO)
 ]
}
